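# Editor utility widget that creates ambient audio volumes around selected actors, combines existing volumes, and updates selected ones.
# Geometry Script merges selected meshes or volume bounds, derives box or convex-hull shapes, and spawns volumes pivoted at bottom-center; combining destroys source volumes.
# Ambient audio volumes receive priority, crossfade time and an optional bank asset, with presets setting the first two.
# Selection buttons only fire dispatchers; sphere and dome builders are unwired.
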EditorUtilityWidgetBlueprint /FortTechAudioTools/Tools/AudioVolumeCreator/EUW_AudioVolumeCreator
graphs [
 AddAmbientBankAssetToVolume {nodes [
   n1 = AddAmbientBankAssetToVolume!() -> (then, GeneratedVolume:Actor Object Reference) {is=title K2Node_FunctionEntry}
   > n2 = Cast To FortAmbientAudioVolume!(execute, Object=@n1) -> (then, CastFailed, AsAmbient Volume:Ambient Volume Object Reference) {is=title K2Node_DynamicCast}
   n3 = GetComponentByClass!(self:Actor Object Reference=@n2, ComponentClass:Actor Component Class Reference=/Script/AmbientAudio.AmbientAudioComponent) -> (ReturnValue:Ambient Audio Component Object Reference)
   > Set Priority!(execute, Priority:int=Get VolumePriority() -> (VolumePriority:int=0), self:Ambient Audio Component Object Reference=@n3) -> (then, Output_Get:int=0)
   > Set CrossfadeTime!(execute, CrossfadeTime:float=Get Crossfade Time() -> (Crossfade Time:float), self:Ambient Audio Component Object Reference=@n3) -> (then, Output_Get:float)
   > n4 = Get!(execute:Exec) -> (then:Exec, else:Exec, AmbientBankAsset:Ambient Audio Entry Object Reference)
   > Set AmbientAsset!(execute, AmbientAsset:Ambient Audio Entry Object Reference=@n4, self:Ambient Audio Component Object Reference=@n3) -> (then, Output_Get:Ambient Audio Entry Object Reference)
  ]}
 CreateBoxMesh {nodes [
   CreateBoxMesh() {is=title K2Node_FunctionEntry}
   > n1 = CreateDynamicMeshPool!(execute) -> (then, ReturnValue:Dynamic Mesh Pool Object Reference)
   > n2 = RequestMesh!(execute, self:Dynamic Mesh Pool Object Reference=@n1) -> (then, ReturnValue:Dynamic Mesh Object Reference)
   n3 = Break Box!(Box:"Box Structure (by ref)"=Get CombinedMeshBoundingBox() -> (CombinedMeshBoundingBox:Box Structure)) -> (Min_X:double, Min_Y:double, Min_Z:double, Max_X:double, Max_Y:double, Max_Z:double, IsValid:bool=~) {is=title K2Node_BreakStruct}
   > n4 = AppendBox!(
    execute
    TargetMesh:Dynamic Mesh Object Reference=@n2
    PrimitiveOptions_PolygroupMode:EGeometryScriptPrimitivePolygroupMode Enum=PerFace
    PrimitiveOptions_bFlipOrientation:bool=False
    PrimitiveOptions_UVMode:EGeometryScriptPrimitiveUVMode Enum=Uniform
    PrimitiveOptions_MaterialID:int=0
    Transform_Location:Vector=Get PivotLocation() -> (PivotLocation:Vector)
    Transform_Rotation:Rotator="0, 0, 0"
    Transform_Scale:Vector="1,1,1"
    DimensionX:float=float - float(A=@n3.Max_X, B=@n3)
    DimensionY:float=float - float(A=@n3.Max_Y, B=@n3.Min_Y)
    DimensionZ:float=float - float(A=@n3.Max_Z, B=@n3.Min_Z)
    StepsX:int=0
    StepsY:int=0
    StepsZ:int=0
    Origin:EGeometryScriptPrimitiveOriginMode Enum=Base
   ) -> (then, ReturnValue:Dynamic Mesh Object Reference)
   > n5 = Set BoxMesh(BoxMesh:Dynamic Mesh Object Reference=@n4) -> (Output_Get:Dynamic Mesh Object Reference)
   > n6 = TransformMesh!(
    execute
    TargetMesh:Dynamic Mesh Object Reference=@n5
    Transform_Location:Vector=vector - vector(B=Get PivotLocation())
    Transform_Rotation:Rotator="0, 0, 0"
    Transform_Scale:Vector="1,1,1"
    bFixOrientationForNegativeScale:bool=true
   ) -> (then, ReturnValue:Dynamic Mesh Object Reference)
   > ApplyMeshOffset!(
    execute
    TargetMesh:Dynamic Mesh Object Reference=@n6
    Options_OffsetDistance:float=1
    Options_bFixedBoundary:bool=False
    Options_SolveSteps:int=0
    Options_SmoothAlpha:float=0.1
    Options_bReprojectDuringSmoothing:bool=False
    Options_BoundaryAlpha:float=0.2
   ) -> (then, ReturnValue:Dynamic Mesh Object Reference)
  ]}
 CreateConvexHullMesh {nodes [
   CreateConvexHullMesh() {is=title K2Node_FunctionEntry}
   > n1 = RequestMesh(self=Get DynamicMeshPool() -> (DynamicMeshPool:Dynamic Mesh Pool Object Reference))
   > n2 = ComputeMeshConvexHull!(
    execute
    TargetMesh:Dynamic Mesh Object Reference=Get CombinedMesh() -> (CombinedMesh:Dynamic Mesh Object Reference)
    CopyToMesh:"Dynamic Mesh Object Reference (by ref)"=@n1
    Selection:Mesh Selection Structure
    Options_bPrefilterVertices:bool=True
    Options_PrefilterGridResolution:int=128
    Options_SimplifyToFaceCount:int=Get FaceCount() -> (FaceCount:int=0)
    Options_GeometricToleranceAfterTarget:float=0
    Debug:Geometry Script Debug Object Reference
   ) -> (then, CopyToMeshOut:Dynamic Mesh Object Reference, ReturnValue:Dynamic Mesh Object Reference)
   > n3 = Set HullMesh(HullMesh:Dynamic Mesh Object Reference=@n2) -> (Output_Get:Dynamic Mesh Object Reference)
   > n4 = TransformMesh(
    TargetMesh=@n3
    Transform_Location=vector - vector(B=Get PivotLocation())
    Transform_Rotation="0, 0, 0"
    Transform_Scale="1,1,1"
    bFixOrientationForNegativeScale=true
   )
   > ApplyMeshOffset(
    TargetMesh=@n4
    Options_OffsetDistance=1
    Options_bFixedBoundary=False
    Options_SolveSteps=0
    Options_SmoothAlpha=0.1
    Options_bReprojectDuringSmoothing=False
    Options_BoundaryAlpha=0.2
   )
  ]}
 CreateDomeMesh {nodes [
   CreateDomeMesh() {is=title K2Node_FunctionEntry}
   > n1 = CreateDynamicMeshPool()
   > n2 = RequestMesh(self=@n1)
   n3 = Get CombinedMeshBoundingBox()
   n4 = Break Box(Box=@n3) -> (IsValid=~)
   n5 = Vector Length!(
    A_X:double=float * float(A=float - float(A=@n4.Max_X, B=@n4), B=0.5) {is=title K2Node_PromotableOperator}
    A_Y:double=float * float(A=float - float(A=@n4.Max_Y, B=@n4.Min_Y), B=0.5)
    A_Z:double=float * float(A=float - float(A=@n4.Max_Z, B=@n4.Min_Z), B=0.5)
   ) -> (ReturnValue)
   n6 = Get PivotLocation()
   > n7 = AppendSphereBox!(
    execute
    TargetMesh:Dynamic Mesh Object Reference=@n2
    PrimitiveOptions:Geometry Script Primitive Options Structure
    Transform_Location:Vector=@n6
    Transform_Rotation:Rotator="0, 0, 0"
    Transform_Scale:Vector="1,1,1"
    Radius:float=@n5
    StepsX:int=6
    StepsY:int=6
    StepsZ:int=6
    Origin:EGeometryScriptPrimitiveOriginMode Enum=Center
   ) -> (then, ReturnValue:Dynamic Mesh Object Reference)
   > n8 = CreateDynamicMeshPool()
   > n9 = RequestMesh(self=@n8)
   n10 = vector - vector!(A=@n6, B_X:double=0, B_Y:double=0, B_Z:double=float * float(A=@n5, B=0.5)) -> (ReturnValue)
   n11 = float * float(A=@n5, B=2)
   > n12 = AppendBox!(
    execute
    TargetMesh=@n9
    PrimitiveOptions:Geometry Script Primitive Options Structure
    Transform_Location=@n10
    Transform_Rotation="0, 0, 0"
    Transform_Scale="1,1,1"
    DimensionX=@n11
    DimensionY=@n11
    DimensionZ=@n11
    StepsX=0
    StepsY=0
    StepsZ=0
    Origin=Base
   ) -> (then, ReturnValue)
   > n13 = Set SubtractionMesh(SubtractionMesh:Dynamic Mesh Object Reference=@n12) -> (Output_Get:Dynamic Mesh Object Reference)
   > n14 = ApplyMeshBoolean!(
    execute
    TargetMesh:Dynamic Mesh Object Reference=@n7
    TargetTransform:Transform="To Transform (Vector)"!(InLocation:Vector=Get PivotLocation()) -> (ReturnValue:Transform)
    ToolMesh:Dynamic Mesh Object Reference=@n13
    ToolTransform:Transform="To Transform (Vector)"(InLocation=@n10)
    Operation:EGeometryScriptBooleanOperation Enum=Subtract
    Options_bFillHoles:bool=True
    Options_bSimplifyOutput:bool=True
    Options_SimplifyPlanarTolerance:float=0.01
    Options_bAllowEmptyResult:bool=False
    Options_OutputTransformSpace:EGeometryScriptBooleanOutputSpace Enum=TargetTransformSpace
   ) -> (then, ReturnValue:Dynamic Mesh Object Reference)
   > n15 = CompactMesh!(execute, TargetMesh:Dynamic Mesh Object Reference=@n14) -> (then, ReturnValue:Dynamic Mesh Object Reference)
   > n16 = RecomputeNormals!(
    execute
    TargetMesh:Dynamic Mesh Object Reference=@n15
    CalculateOptions:Geometry Script Calculate Normals Options Structure
    bDeferChangeNotifications:bool=false
   ) -> (then, ReturnValue:Dynamic Mesh Object Reference)
   > n17 = Set SphereMesh(SphereMesh:Dynamic Mesh Object Reference=@n16) -> (Output_Get:Dynamic Mesh Object Reference)
   > n18 = TransformMesh(
    TargetMesh=@n17
    Transform_Location=vector - vector!(A, B:Vector=Get PivotLocation()) -> (ReturnValue)
    Transform_Rotation="0, 0, 0"
    Transform_Scale="1,1,1"
    bFixOrientationForNegativeScale=true
   )
   > ApplyMeshOffset(
    TargetMesh=@n18
    Options_OffsetDistance=1
    Options_bFixedBoundary=False
    Options_SolveSteps=0
    Options_SmoothAlpha=0.1
    Options_bReprojectDuringSmoothing=False
    Options_BoundaryAlpha=0.2
   )
   "Get Center (Box)"!(InBox:"Box Structure (by ref)"=@n3) -> (ReturnValue:Vector) {is=title K2Node_CallFunction}
  ]}
 CreatePivotFromBounds {nodes [
   CreatePivotFromBounds() {is=title K2Node_FunctionEntry}
   n1 = Break Box(Box=Get CombinedMeshBoundingBox()) -> (IsValid=~)
   n2 = float + float(A=@n1.Min_Y, B=@n1.Max_Y)
   n3 = float + float(A=@n1, B=@n1.Max_X)
   > Return Node!(execute, ReturnValue:Vector=MakeVector(X=float * float(A=@n3, B=0.5), Y=float * float(A=@n2, B=0.5), Z=@n1.Min_Z))
  ]}
 CreateSphereMesh {nodes [
   CreateSphereMesh() {is=title K2Node_FunctionEntry}
   > n1 = CreateDynamicMeshPool()
   > n2 = RequestMesh(self=@n1)
   n3 = Break Box(Box=Get CombinedMeshBoundingBox()) -> (IsValid=~)
   n4 = float - float(A=@n3.Max_X, B=@n3)
   n5 = float - float(A=@n3.Max_Z, B=@n3.Min_Z)
   n6 = float - float(A=@n3.Max_Y, B=@n3.Min_Y)
   > n7 = AppendSphereBox(
    TargetMesh=@n2
    Transform_Location=Get PivotLocation()
    Transform_Rotation="0, 0, 0"
    Transform_Scale="1,1,1"
    Radius=Vector Length(A_X=float * float(A=@n4, B=0.5), A_Y=float * float(A=@n6, B=0.5), A_Z=float * float(A=@n5, B=0.5))
    StepsX=6
    StepsY=6
    StepsZ=6
    Origin=Center
   )
   > n8 = Set SphereMesh(SphereMesh=@n7)
   > n9 = TransformMesh(
    TargetMesh=@n8
    Transform_Location=vector - vector(B=Get PivotLocation())
    Transform_Rotation="0, 0, 0"
    Transform_Scale="1,1,1"
    bFixOrientationForNegativeScale=true
   )
   > ApplyMeshOffset(
    TargetMesh=@n9
    Options_OffsetDistance=1
    Options_bFixedBoundary=False
    Options_SolveSteps=0
    Options_SmoothAlpha=0.1
    Options_bReprojectDuringSmoothing=False
    Options_BoundaryAlpha=0.2
   )
  ]}
 EventGraph {nodes [
   "On Clicked (CreateVolume_Button)"() {is=title K2Node_ComponentBoundEvent}
   > Set IsCombinedVolume(IsCombinedVolume:bool=false) -> (Output_Get:bool)
   > n1 = GetSelectedActors()
   > Set SelectedActors(SelectedActors:Array of Actor Object References=@n1) -> (Output_Get:Array of Actor Object References)
   > Compare Int!(Exec:Exec, Input:int=Length(TargetArray=@n1), CompareWith:int) -> (">":Exec, "==":Exec, "<":Exec) {is=title K2Node_MacroInstance}
   > n2 = CreateDynamicMeshPool()
   > n3 = Set DynamicMeshPool(DynamicMeshPool:Dynamic Mesh Pool Object Reference=@n2) -> (Output_Get:Dynamic Mesh Pool Object Reference)
   > n4 = RequestMesh(self=@n3)
   > Set CombinedMesh(CombinedMesh:Dynamic Mesh Object Reference=@n4) -> (Output_Get:Dynamic Mesh Object Reference)
   > n5 = For Each Loop(Array:Array of Actor Object References=Get SelectedActors() -> (SelectedActors:Array of Actor Object References)) -> (Array Element:Actor Object Reference)
   Set CombinedMeshBoundingBox(
    execute=@n5.Completed
    CombinedMeshBoundingBox:Box Structure=GetMeshBoundingBox!(TargetMesh:Dynamic Mesh Object Reference=Get CombinedMesh()) -> (ReturnValue:Box Structure)
   ) -> (Output_Get:Box Structure)
   > n6 = CreatePivotFromBounds!(execute, self) -> (then, ReturnValue:Vector) {is=title K2Node_CallFunction}
   > Set PivotLocation(PivotLocation:Vector=@n6) -> (Output_Get:Vector)
   > n7 = Switch on AudioVolumeCreator_VolumeShape!(
    execute
    Selection:AudioVolumeCreator_VolumeShape Enum=Get VolumeShapeCalculation() -> (VolumeShapeCalculation:AudioVolumeCreator_VolumeShape Enum=NewEnumerator0)
   ) -> (NewEnumerator0:Exec, NewEnumerator1:Exec) {is=title K2Node_SwitchEnum}
   > n8 = CreateBoxMesh() {is=title K2Node_CallFunction}
   SetVolumeName(execute=[@n10.then, @n8.then])
   > n9 = GenerateVolume!(
    execute
    self
    FromDynamicMesh:Dynamic Mesh Object Reference=Select!(
     NewEnumerator0:Dynamic Mesh Object Reference=Get BoxMesh() -> (BoxMesh:Dynamic Mesh Object Reference)
     NewEnumerator1:Dynamic Mesh Object Reference=Get HullMesh() -> (HullMesh:Dynamic Mesh Object Reference)
     Index:AudioVolumeCreator_VolumeShape Enum=Get VolumeShapeCalculation() -> (VolumeShapeCalculation=NewEnumerator0)
    ) -> (ReturnValue:Dynamic Mesh Object Reference)
    ActorTransform_Location:Vector=Get PivotLocation()
    BaseActorName:String=Get GeneratedVolumeName() -> (GeneratedVolumeName:String)
    Options_VolumeType:Volume Class Reference=Get VolumeType() -> (VolumeType:Volume Class Reference)
   ) -> (then, GeneratedVolume:Volume Object Reference)
   > OnVolumeGenerated!(execute, self, GeneratedVolume:Actor Object Reference=@n9) -> (then)
   n10 = CreateConvexHullMesh(execute=@n7.NewEnumerator1) {is=title K2Node_CallFunction}
   n11 = RequestMesh(execute=@n5.LoopBody, self=Get DynamicMeshPool())
   > n12 = CopyMeshFromComponent!(
    execute
    Component:Scene Component Object Reference=Get RootComponent!(self:Actor Object Reference=@n5) -> (RootComponent:Scene Component Object Reference)
    ToDynamicMesh:Dynamic Mesh Object Reference=@n11
    Options_bWantNormals:bool=True
    Options_bWantTangents:bool=True
    Options_bWantInstanceColors:bool=False
    Options_RequestedLOD:Geometry Script Mesh Read LOD Structure=(LODType=MaxAvailable,LODIndex=0)
    bTransformToWorld:bool=false
   ) -> (Failure:Exec, Success:Exec, LocalToWorld:Transform, ReturnValue:Dynamic Mesh Object Reference)
   AppendMesh!(
    execute=@n12.Success
    TargetMesh:Dynamic Mesh Object Reference=Get CombinedMesh()
    AppendMesh:Dynamic Mesh Object Reference=@n12.ReturnValue
    AppendTransform:Transform=@n12
    bDeferChangeNotifications:bool=false
    AppendOptions:Geometry Script Append Mesh Options Structure=(CombineMode=EnableAllMatching)
   ) -> (then, ReturnValue:Dynamic Mesh Object Reference)
   "On Clicked (SelectLastVolume_button)"() {is=title K2Node_ComponentBoundEvent}
   > Call Select Last Volume!(
    execute:Exec
    self:EUW Audio Volume Creator Object Reference
    GeneratedVolume:Actor Object Reference=Get GeneratedVolume() -> (GeneratedVolume:Volume Object Reference)
   ) -> (then:Exec) {is=title K2Node_CallDelegate}
   "On Clicked (CombineVolumes_Button)"() {is=title K2Node_ComponentBoundEvent}
   > Set IsCombinedVolume(IsCombinedVolume=true)
   > n13 = GetSelectedActors()
   > Set SelectedActors(SelectedActors=@n13)
   > Compare Int(Input=Length(TargetArray=@n13))
   > n14 = CreateDynamicMeshPool()
   > n15 = Set DynamicMeshPool(DynamicMeshPool=@n14)
   > n16 = RequestMesh(self=@n15)
   > Set CombinedMesh(CombinedMesh=@n16)
   > n17 = Switch on AudioVolumeCreator_VolumeShape(Selection=Get VolumeShapeCalculation() -> (VolumeShapeCalculation=NewEnumerator0))
   > n18 = For Each Loop(Array=Get SelectedActors())
   Set CombinedMeshBoundingBox(execute=@n18.Completed, CombinedMeshBoundingBox=GetMeshBoundingBox(TargetMesh=Get CombinedMesh()))
   > CreatePivotFromBounds()
   > n19 = TransformMesh(
    TargetMesh=Get CombinedMesh()
    Transform_Location=vector - vector(B=Get PivotLocation())
    Transform_Rotation="0, 0, 0"
    Transform_Scale="1,1,1"
    bFixOrientationForNegativeScale=true
   )
   > n20 = ApplyMeshOffset(
    TargetMesh=@n19
    Options_OffsetDistance=1
    Options_bFixedBoundary=False
    Options_SolveSteps=0
    Options_SmoothAlpha=0.1
    Options_bReprojectDuringSmoothing=False
    Options_BoundaryAlpha=0.2
   )
   > SetVolumeName()
   > n21 = GenerateVolume(
    FromDynamicMesh=@n20
    ActorTransform_Location=Get PivotLocation()
    BaseActorName=Get GeneratedVolumeName()
    Options_VolumeType=Select!(
     Option 0:Volume Class Reference=Get VolumeType()
     Option 1:Volume Class Reference=Get NewCombinedVolumeType() -> (NewCombinedVolumeType:Volume Class Reference)
     Index:bool=Get Override Previous Volume Type() -> (Override Previous Volume Type:bool)
    ) -> (ReturnValue:Volume Class Reference)
   )
   > OnVolumeGenerated(GeneratedVolume=@n21)
   Cast To Volume!(execute=@n18.LoopBody, Object=@n18) -> (then, CastFailed, AsVolume:Volume Object Reference) {is=title K2Node_DynamicCast}
   n22 = Get Actor Bounds!(self=@n18, bOnlyCollidingComponents=false, bIncludeFromChildActors) -> (Origin, BoxExtent_X:double, BoxExtent_Y:double, BoxExtent_Z:double)
   n23 = BreakVector(InVec=@n22)
   > AppendBox!(
    execute
    TargetMesh=Get CombinedMesh()
    PrimitiveOptions
    Transform_Location_X:double=@n23
    Transform_Location_Y:double=@n23.Y
    Transform_Location_Z:double=float - float(A=@n23.Z, B=@n22.BoxExtent_Z)
    Transform_Rotation="0, 0, 0"
    Transform_Scale="1,1,1"
    DimensionX=float * float(A=@n22.BoxExtent_X, B=2)
    DimensionY=float * float(A=@n22.BoxExtent_Y, B=2)
    DimensionZ=float * float(A=@n22.BoxExtent_Z, B=2)
    StepsX=0
    StepsY=0
    StepsZ=0
    Origin=Base
   ) -> (then, ReturnValue)
   > Destroy Actor(self=@n18)
   n24 = For Each Loop(Exec=@n17.NewEnumerator1, Array=Get SelectedActors())
   Set CombinedMeshBoundingBox(execute=@n24.Completed, CombinedMeshBoundingBox=GetMeshBoundingBox(TargetMesh=Get CombinedMesh()))
   > n25 = CreatePivotFromBounds()
   > Set PivotLocation(PivotLocation=@n25)
   > CreateConvexHullMesh()
   > n26 = GetEditorWorld(self=Get UnrealEditorSubsystem())
   > SetVolumeName()
   > n27 = Create New Volume From Mesh!(
    execute
    WorldContextObject:Object Reference=@n26
    FromDynamicMesh:Dynamic Mesh Object Reference=Get HullMesh()
    ActorTransform_Location:Vector=Get PivotLocation()
    ActorTransform_Rotation:Rotator="0, 0, 0"
    ActorTransform_Scale:Vector="1,1,1"
    BaseActorName:String=Get GeneratedVolumeName()
    Options_VolumeType:Volume Class Reference=Select(Option 0=Get VolumeType(), Option 1=Get NewCombinedVolumeType(), Index=Get Override Previous Volume Type())
    Options_bAutoSimplify:bool=True
    Options_MaxTriangles:int=250
   ) -> (Failure:Exec, Success:Exec, ReturnValue:Volume Object Reference)
   n28 = Set GeneratedVolume(execute=@n27.Success, GeneratedVolume:Volume Object Reference=@n27) -> (Output_Get:Volume Object Reference)
   > ReturnAllMeshes(self:Dynamic Mesh Pool Object Reference=Get DynamicMeshPool())
   > OnVolumeGenerated(GeneratedVolume=@n28)
   Cast To Volume(execute=@n24.LoopBody, Object=@n24)
   > n29 = RequestMesh(self=Get DynamicMeshPool())
   > n30 = CopyMeshFromComponent(
    Component=Get RootComponent(self=@n24)
    ToDynamicMesh=@n29
    Options_bWantNormals=True
    Options_bWantTangents=True
    Options_bWantInstanceColors=False
    Options_RequestedLOD=(LODType=MaxAvailable,LODIndex=0)
    bTransformToWorld=false
   )
   AppendMesh(
    execute=@n30.Success
    TargetMesh=Get CombinedMesh()
    AppendMesh=@n30.ReturnValue
    AppendTransform=@n30
    bDeferChangeNotifications=false
    AppendOptions=(CombineMode=EnableAllMatching)
   )
   > Destroy Actor(self=@n24)
   "On Clicked (SelectAllVolumes_Button)"() {is=title K2Node_ComponentBoundEvent}
   > Call On Volume Generated Notify!(execute:Exec, self:EUW Audio Volume Creator Object Reference, GeneratedVolume:Actor Object Reference) -> (then:Exec) {is=title K2Node_CallDelegate}
   "On Clicked (PreviouslySelectedActors_Button)"() {is=title K2Node_ComponentBoundEvent}
   > Call Select Previously Selected Actors!(execute:Exec, self:EUW Audio Volume Creator Object Reference) -> (then:Exec) {is=title K2Node_CallDelegate}
   n31 = "On Property Changed (VolumeSettings)"!() -> (OutputDelegate, then, PropertyName:Name=None) {is=title K2Node_ComponentBoundEvent}
   > n32 = Branch(Condition="Equal (Name)"(A=@n31.PropertyName, B=VolumeShapeCalculation))
   > n33 = Switch on AudioVolumeCreator_VolumeShape(Selection=Get VolumeShapeCalculation() -> (VolumeShapeCalculation=NewEnumerator0))
   > Set FaceCount(FaceCount:int=6) -> (Output_Get:int=0)
   Set FaceCount(execute=@n33.NewEnumerator1, FaceCount=12) -> (Output_Get=0)
   Branch(execute=@n32.else, Condition="Equal (Name)"(A=@n31.PropertyName, B=AmbientBankPriorityPreset))
   > n34 = Switch on AudioVolumeCreator_AmbiencePriority_Presets!(
    execute
    Selection:AudioVolumeCreator_AmbiencePriority_Presets Enum=Get AmbientBankPriorityPreset() -> (AmbientBankPriorityPreset:AudioVolumeCreator_AmbiencePriority_Presets Enum=NewEnumerator0)
   ) -> (NewEnumerator0:Exec, NewEnumerator1:Exec, NewEnumerator4:Exec, NewEnumerator5:Exec) {is=title K2Node_SwitchEnum}
   > Set VolumePriority(VolumePriority:int=1) -> (Output_Get:int=0)
   > Set Crossfade Time(Crossfade Time:float=10) -> (Output_Get:float)
   Set VolumePriority(execute=@n34.NewEnumerator1, VolumePriority=2) -> (Output_Get=0)
   > Set Crossfade Time(Crossfade Time=10)
   Set VolumePriority(execute=@n34.NewEnumerator4, VolumePriority=3) -> (Output_Get=0)
   > Set Crossfade Time(Crossfade Time=10)
   Set VolumePriority(execute=@n34.NewEnumerator5, VolumePriority=10) -> (Output_Get=0)
   > Set Crossfade Time(Crossfade Time=5)
   "On Clicked (UpdateSelectedVolume)"() {is=title K2Node_ComponentBoundEvent}
   > n35 = GetSelectedActors()
   > n36 = Set SelectedActors(SelectedActors=@n35)
   > Branch(Condition=Is Not Empty!(TargetArray:Array of Actor Object References=@n35) -> (ReturnValue:bool))
   > n37 = For Each Loop(Array=@n36)
   > n38 = Cast To Volume(Object=@n37)
   > AddAmbientBankAssetToVolume!(execute, self, GeneratedVolume:Actor Object Reference=@n38) -> (then) {is=title K2Node_CallFunction}
   n39 = OnVolumeGenerated!() -> (OutputDelegate, then, GeneratedVolume:Actor Object Reference) {is=title K2Node_CustomEvent}
   > Add(
    TargetArray:Array of Actor Object References=Get GeneratedVolumeArray() -> (GeneratedVolumeArray:Array of Actor Object References)
    NewItem:"Actor Object Reference (by ref)"=@n39.GeneratedVolume
   ) {is=title K2Node_CallArrayFunction}
   > Call On Volume Generated Notify(GeneratedVolume=@n39.GeneratedVolume)
   > AddAmbientBankAssetToVolume(GeneratedVolume=@n39.GeneratedVolume)
   > Set SelectionContainsVolume(SelectionContainsVolume:bool=false) -> (Output_Get:bool)
   > n40 = For Each Loop(Array=Get SelectedActors())
   n41 = Get!(execute=@n40.Completed) -> (then, else, SelectionContainsVolume:bool)
   > Clear(TargetArray:Array of Actor Object References=Get SelectedActors())
   Clear(execute=@n41.else, TargetArray=Get PreviouslySelectedActors() -> (PreviouslySelectedActors:Array of Actor Object References))
   > Set PreviouslySelectedActors(PreviouslySelectedActors:Array of Actor Object References=Get SelectedActors()) -> (Output_Get:Array of Actor Object References)
   > Clear(TargetArray=Get SelectedActors())
   Branch(
    execute=@n40.LoopBody
    Condition=Contains Item(TargetArray:Array of Actor Object References=Get GeneratedVolumeArray(), ItemToFind:"Actor Object Reference (by ref)"=@n40)
   )
   > Remove Item!(execute, TargetArray:Array of Actor Object References=Get SelectedActors(), Item:"Actor Object Reference (by ref)"=@n40) -> (then, ReturnValue:bool)
   > Set SelectionContainsVolume(SelectionContainsVolume=true)
   Event PreConstruct()
   > SetObject(self=Get VolumeSettings() -> (VolumeSettings:Details View Object Reference), NewObject=Self-Reference())
   n42 = Get AmbientBankAssetField() -> (AmbientBankAssetField:Single Property View Object Reference)
   > SetObject(self=@n42, NewObject=Self-Reference())
   > SetPropertyName!(execute, self:Single Property View Object Reference=@n42, NewPropertyName:Name=AmbientBankAsset) -> (then)
   CreateSphereMesh() {is=title K2Node_CallFunction}
   CreateDomeMesh() {is=title K2Node_CallFunction}
  ]}
 GenerateVolume {nodes [
   n1 = GenerateVolume!() -> (
    then
    FromDynamicMesh:Dynamic Mesh Object Reference
    ActorTransform_Location:Vector=~
    BaseActorName:String
    Options_VolumeType:Volume Class Reference
   ) {is=title K2Node_FunctionEntry}
   > n2 = GetEditorWorld(self=Get UnrealEditorSubsystem())
   > n3 = Create New Volume From Mesh(
    WorldContextObject=@n2
    FromDynamicMesh=@n1
    ActorTransform_Location=@n1.ActorTransform_Location
    ActorTransform_Rotation="0, 0, 0"
    ActorTransform_Scale="1,1,1"
    BaseActorName=@n1.BaseActorName
    Options_VolumeType=@n1.Options_VolumeType
    Options_bAutoSimplify=True
    Options_MaxTriangles=250
   )
   n4 = Set GeneratedVolume(execute=@n3.Success, GeneratedVolume=@n3)
   > ReturnAllMeshes(self=Get DynamicMeshPool())
   > Return Node!(execute, GeneratedVolume:Volume Object Reference=@n4)
  ]}
 OnRep_VolumeType {nodes [OnRep_VolumeType() {is=title K2Node_FunctionEntry}]}
 OnVolumeGenerated_Notify {nodes [OnVolumeGenerated_Notify!() -> (then, GeneratedVolume:Actor Object Reference) {is=title K2Node_FunctionEntry}]}
 SelectLastVolume {nodes [SelectLastVolume!() -> (then, GeneratedVolume:Actor Object Reference) {is=title K2Node_FunctionEntry}]}
 SelectPreviouslySelectedActors {nodes [SelectPreviouslySelectedActors() {is=title K2Node_FunctionEntry}]}
 SetVolumeName {nodes [
   SetVolumeName() {is=title K2Node_FunctionEntry}
   > n1 = Branch(Condition=Get IsCombinedVolume() -> (IsCombinedVolume:bool))
   n2 = Get VolumeType()
   > Set GeneratedVolumeName(GeneratedVolumeName:String=Append!(A=Get Class Display Name(Class=@n2), B=_, C:String=Get VolumeName() -> (VolumeName:String)) -> (ReturnValue)) -> (Output_Get:String)
   n3 = Select(Option 0=Get VolumeType(), Option 1=Get NewCombinedVolumeType(), Index=Get Override Previous Volume Type())
   Set GeneratedVolumeName(execute=@n1.else, GeneratedVolumeName=Append(A=Get Class Display Name(Class=@n3), B=_, C=Get VolumeName()))
  ]}
]
variables [
 AmbientBankAsset {type=AmbientAudioDataAsset, default=null}
 AmbientBankPriorityPreset {type=AudioVolumeCreator_AmbiencePriority_Presets, default=Biome}
 BoxMesh {type=DynamicMesh, default=null}
 CombinedMesh {type=DynamicMesh, default=null}
 CombinedMeshBoundingBox {type=Box, default {isValid=false, max=<0 0 0>, min=<0 0 0>}, shapeType=~}
 Crossfade Time {type=number, default=10}
 DynamicMeshPool {type=DynamicMeshPool, default=null}
 FaceCount {type=integer, default=6}
 GeneratedVolume {type=Volume, default=null}
 GeneratedVolumeArray {type="array<Actor>"}
 GeneratedVolumeName {type=string, default=""}
 HullMesh {type=DynamicMesh, default=null}
 IsCombinedVolume {type=boolean}
 NewCombinedVolumeType {type=Class, default=/Script/FortniteGame.FortAmbientAudioVolume}
 OnVolumeGenerated_Notify {default=()}
 Override Previous Volume Type {type=boolean}
 PivotLocation {type=Vector}
 PreviouslySelectedActors {type="array<Actor>"}
 SelectedActors {type="array<Actor>"}
 SelectionContainsVolume {type=boolean}
 SelectLastVolume {default=()}
 SelectPreviouslySelectedActors {default=()}
 SphereMesh {type=DynamicMesh, default=null}
 SubtractionMesh {type=DynamicMesh, default=null}
 VolumeName {type=string, default=GeneratedVolume}
 VolumePriority {type=integer}
 VolumeShapeCalculation {type=AudioVolumeCreator_VolumeShape, default=Box}
 VolumeType {type=Class, default=&FortAmbientAudioVolume}
]
widgets {
 info {rootWidgetClass=&GridPanel, widgetCount=20}
 widgets [
  GridPanel_0 {widgetClassPath=&GridPanel, widget=GridPanel_0}
  ScaleBox_38 {widgetClassPath=&ScaleBox, parent=GridPanel_0, widget=ScaleBox_38, slot {class=GridSlot}}
  VerticalBox_55 {widgetClassPath=&VerticalBox, parent=GridPanel_0, widget=VerticalBox_55, slot {class=GridSlot, properties {row=1}}}
  HorizontalBox_236 {widgetClassPath=&HorizontalBox, parent=VerticalBox_55, widget=HorizontalBox_236, slot {class=VerticalBoxSlot}}
  CreateVolume_Button {
   widgetClassPath=&EditorUtilityButton
   parent=HorizontalBox_236
   widget=CreateVolume_Button
   properties {toolTipText=Creates a volume from the meshes selected in the viewport and updates it with the provided ambient bank settings.}
   slot {class=HorizontalBoxSlot, properties {horizontalAlignment=HAlign_Left}}
  }
  TextBlock_60 {
   widgetClassPath=&TextBlock
   parent=CreateVolume_Button
   widget=TextBlock_60
   properties {text=Create Volume}
   slot {class=ButtonSlot}
  }
  CombineVolumes_Button {
   widgetClassPath=&EditorUtilityButton
   parent=HorizontalBox_236
   widget=CombineVolumes_Button
   properties {toolTipText=Combines the selected volumes and updates the new volume with the provided ambient bank settings.}
   slot {class=HorizontalBoxSlot}
  }
  CombineVolumes {
   widgetClassPath=&TextBlock
   parent=CombineVolumes_Button
   widget=CombineVolumes
   properties {text=Combine Volumes}
   slot {class=ButtonSlot}
  }
  ScaleBox_98 {widgetClassPath=&ScaleBox, parent=HorizontalBox_236, widget=ScaleBox_98, slot {class=HorizontalBoxSlot}}
  UpdateSelectedVolume {
   widgetClassPath=&EditorUtilityButton
   parent=VerticalBox_55
   widget=UpdateSelectedVolume
   properties {toolTipText="Updates the volume priority, crossfade time, and ambient asset if one is provided."}
   slot {class=VerticalBoxSlot}
  }
  UpdateVolume {
   widgetClassPath=&TextBlock
   parent=UpdateSelectedVolume
   widget=UpdateVolume
   properties {text="Update Selected Volume(s)"}
   slot {class=ButtonSlot}
  }
  AmbientBankAssetField {
   widgetClassPath=&SinglePropertyView
   parent=VerticalBox_55
   widget=AmbientBankAssetField
   slot {class=VerticalBoxSlot}
  }
  VolumeSettings {
   widgetClassPath=&DetailsView
   parent=VerticalBox_55
   widget=VolumeSettings
   properties {categoriesToShow [Ambient Bank Settings, Volume Settings, Combined Volume Overrides]}
   slot {class=VerticalBoxSlot}
  }
  VerticalBox_175 {widgetClassPath=&VerticalBox, parent=VerticalBox_55, widget=VerticalBox_175, slot {class=VerticalBoxSlot}}
  SelectAllVolumes_Button {
   widgetClassPath=&EditorUtilityButton
   parent=VerticalBox_175
   widget=SelectAllVolumes_Button
   properties {toolTipText=Selects all volumes created in the current tooling session.}
   slot {class=VerticalBoxSlot}
  }
  TextBlock_1 {
   widgetClassPath=&TextBlock
   parent=SelectAllVolumes_Button
   widget=TextBlock_1
   properties {text=Select All Created Volumes}
   slot {class=ButtonSlot}
  }
  PreviouslySelectedActors_Button {
   widgetClassPath=&EditorUtilityButton
   parent=VerticalBox_175
   widget=PreviouslySelectedActors_Button
   properties {toolTipText=Selects the last selected meshes. Does not select anything if Combine Volumes was just used.}
   slot {class=VerticalBoxSlot}
  }
  TextBlock_2 {
   widgetClassPath=&TextBlock
   parent=PreviouslySelectedActors_Button
   widget=TextBlock_2
   properties {text=Select Previous Meshes}
   slot {class=ButtonSlot}
  }
  SelectLastVolume_button {
   widgetClassPath=&EditorUtilityButton
   parent=VerticalBox_175
   widget=SelectLastVolume_button
   properties {toolTipText=Selects the last created volume in the session.}
   slot {class=VerticalBoxSlot}
  }
  TextBlock {
   widgetClassPath=&TextBlock
   parent=SelectLastVolume_button
   widget=TextBlock
   properties {text=Select Last Created Volume}
   slot {class=ButtonSlot}
  }
 ]
}
