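# Editor scriptable tool for creating audio volumes, hosting the EUW_AudioVolumeCreator widget in its tool header and previewing selection bounds.
# On setup and selection changes it draws a twelve-line box around selected actors, excluding generated volumes after changes, and collapses it when none remain.
# Widget events record and select generated volumes, select the last volume, or restore the previous selection.
# CreateVolume is an empty stub, with generation reported by the widget.
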EditorUtilityBlueprint /FortTechAudioTools/Tools/AudioVolumeCreator/BP_ST_AudioVolumeEditor
graphs [
 CreateVolume {nodes [CreateVolume() {is=title K2Node_FunctionEntry}]}
 DrawBBox {nodes [
   n1 = DrawBBox!() -> (
    then
    BoundOrigin:Vector=~
    BoundExtent:Vector=~
    BoxLines:Array of Scriptable Tool Line Object References
    Color:Color Structure
    Thickness:double=~
    DepthBias:double=~
   ) {is=title K2Node_FunctionEntry}
   > Clear(TargetArray:Array of Vectors=Get CurrentCorners() -> (CurrentCorners:Array of Vectors))
   > UpdateBoxAppearance!(
    execute
    self
    Box Lines:Array of Scriptable Tool Line Object References=@n1.BoxLines
    Color:Color Structure=@n1.Color
    Thickness:float=@n1.Thickness
    DepthBias:float=@n1.DepthBias
   ) -> (then)
   > n2 = For Loop(LastIndex=7)
   n3 = Sequence!(execute=@n2.Completed) -> (
    then_0
    then_1
    then_2:Exec
    then_3:Exec
    then_4:Exec
    then_5:Exec
    then_6:Exec
    then_7:Exec
    then_8:Exec
    then_9:Exec
    then_10:Exec
    then_11:Exec
   )
   > SetLineEdgeBBox!(
    execute
    self
    LineNumber:int=0
    StartPointNumber:int=0
    EndPointNumber:int=1
    BoxLines:Array of Scriptable Tool Line Object References=Get BoxLines() -> (BoxLines:Array of Scriptable Tool Line Object References)
    Corners:Array of Vectors=Get CurrentCorners()
   ) -> (then)
   SetLineEdgeBBox(execute=@n3.then_1, LineNumber=1, StartPointNumber=1, EndPointNumber=6, BoxLines=Get BoxLines(), Corners=Get CurrentCorners())
   SetLineEdgeBBox(execute=@n3.then_10, LineNumber=10, StartPointNumber=6, EndPointNumber=7, BoxLines=Get BoxLines(), Corners=Get CurrentCorners())
   SetLineEdgeBBox(execute=@n3.then_11, LineNumber=11, StartPointNumber=4, EndPointNumber=2, BoxLines=Get BoxLines(), Corners=Get CurrentCorners())
   SetLineEdgeBBox(execute=@n3.then_2, LineNumber=2, StartPointNumber=6, EndPointNumber=2, BoxLines=Get BoxLines(), Corners=Get CurrentCorners())
   SetLineEdgeBBox(execute=@n3.then_3, LineNumber=3, StartPointNumber=2, EndPointNumber=0, BoxLines=Get BoxLines(), Corners=Get CurrentCorners())
   SetLineEdgeBBox(execute=@n3.then_4, LineNumber=4, StartPointNumber=0, EndPointNumber=3, BoxLines=Get BoxLines(), Corners=Get CurrentCorners())
   SetLineEdgeBBox(execute=@n3.then_5, LineNumber=5, StartPointNumber=3, EndPointNumber=5, BoxLines=Get BoxLines(), Corners=Get CurrentCorners())
   SetLineEdgeBBox(execute=@n3.then_6, LineNumber=6, StartPointNumber=5, EndPointNumber=1, BoxLines=Get BoxLines(), Corners=Get CurrentCorners())
   SetLineEdgeBBox(execute=@n3.then_7, LineNumber=7, StartPointNumber=5, EndPointNumber=7, BoxLines=Get BoxLines(), Corners=Get CurrentCorners())
   SetLineEdgeBBox(execute=@n3.then_8, LineNumber=8, StartPointNumber=7, EndPointNumber=4, BoxLines=Get BoxLines(), Corners=Get CurrentCorners())
   SetLineEdgeBBox(execute=@n3.then_9, LineNumber=9, StartPointNumber=4, EndPointNumber=3, BoxLines=Get BoxLines(), Corners=Get CurrentCorners())
   n4 = vector * vector(A=Get BoundExtent() -> (BoundExtent:Vector), B=Make Literal Float(Value=1)) -> (ReturnValue=~)
   n5 = Get BoundOrigin() -> (BoundOrigin:Vector)
   n6 = Get BoundOrigin()
   Add(
    execute=@n2.LoopBody
    TargetArray:Array of Vectors=Get CurrentCorners()
    NewItem:"Vector (by ref)"=GetBoxCorner!(Box_Min:Vector=vector - vector(A=@n6, B=@n4), Box_Max:Vector=vector + vector(A=@n5, B=@n4), CornerIndex:int=@n2) -> (ReturnValue:Vector)
   ) {is=title K2Node_CallArrayFunction}
  ]}
 DrawBox {nodes [
   n1 = DrawBox!() -> (then, Actors:Array of Actor Object References) {is=title K2Node_FunctionEntry}
   > n2 = Branch(Condition="integer > integer"(A=Length(TargetArray=@n1)))
   > n3 = GetActorArrayBounds!(execute, Actors:Array of Actor Object References=@n1, bOnlyCollidingComponents:bool=false) -> (then, Center:Vector, BoxExtent:Vector)
   > DrawBBox!(
    execute
    self
    BoundOrigin:Vector=@n3
    BoundExtent:Vector=@n3.BoxExtent
    BoxLines:Array of Scriptable Tool Line Object References=Get BoundingBoxLines() -> (BoundingBoxLines:Array of Scriptable Tool Line Object References)
    Color:Color Structure=Make Color!(B:Byte=1, G:Byte=0, R:Byte=1, A:Byte=1) -> (Color:Color Structure) {is=title K2Node_MakeStruct}
    Thickness:double=3
    DepthBias:double=0
   ) -> (then) {is=title K2Node_CallFunction}
   DrawBBox(
    execute=@n2.else
    BoundOrigin="0, 0, 0"
    BoundExtent="0, 0, 0"
    BoxLines=Get BoundingBoxLines()
    Color=Make Color(B=0, G=1, R=0, A=1)
    Thickness=3
    DepthBias=0
   )
  ]}
 EventGraph {nodes [
   Event OnScriptSetup()
   > n1 = Sequence!(execute) -> (then_0, then_1, then_2, then_3)
   > K2Node_Composite_1 = Add Widget and Bindings() {is=title K2Node_Composite}
   K2Node_Composite_6 = BindToSelChanged(execute=@n1.then_1) {is=title K2Node_Composite}
   K2Node_Composite_7 = Add Line Set(execute=@n1.then_2) {is=title K2Node_Composite}
   K2Node_Composite_8 = Init BBox(execute=@n1.then_3) {is=title K2Node_Composite}
   Event OnScriptTick()
  ]}
 Add Widget and Bindings {
  path=EventGraph.K2Node_Composite_1.Add Widget and Bindings
  nodes [
   Inputs()
   > n1 = Sequence!(execute) -> (then_0, then_1)
   > n2 = Construct Object from Class(Class=&/EUW_AudioVolumeCreator.EUW_AudioVolumeCreator_C) -> (ReturnValue:EUW Audio Volume Creator Object Reference)
   > SetToolPanelToolHeaderWidget(Widget=@n2)
   > Bind Event to On Volume Generated Notify(self:EUW Audio Volume Creator Object Reference=@n2, Delegate=Create Event()) {is=title K2Node_AddDelegate}
   > Bind Event to Select Last Volume(self:EUW Audio Volume Creator Object Reference=@n2, Delegate=Create Event()) {is=title K2Node_AddDelegate}
   > Bind Event to Select Previously Selected Actors(self:EUW Audio Volume Creator Object Reference=@n2, Delegate=Create Event()) {is=title K2Node_AddDelegate}
   > Set Actions Area(Actions Area:EUW Audio Volume Creator Object Reference=@n2) -> (Output_Get:EUW Audio Volume Creator Object Reference)
   DisplayUserWarningMessage(execute=@n1.then_1, Message="LOCTEXT(\" \")")
   Outputs()
  ]
 }
 BindToSelChanged {
  path=EventGraph.K2Node_Composite_6.BindToSelChanged
  nodes [
   Inputs()
   > n1 = GetSelectionSet(self=Get LevelEditorSubsystem())
   > Bind Event to On Selection Change(self:Typed Element Selection Set Object Reference=@n1, Delegate=Create Event()) {is=title K2Node_AddDelegate}
   Outputs()
  ]
 }
 Add Line Set {
  path=EventGraph.K2Node_Composite_7.Add Line Set
  nodes [
   Inputs()
   > n1 = AddLineSet!(execute, self:Scriptable Interactive Tool Object Reference) -> (then, ReturnValue:Scriptable Tool Line Set Object Reference)
   > Set Line Set(Line Set:Scriptable Tool Line Set Object Reference=@n1) -> (Output_Get:Scriptable Tool Line Set Object Reference)
   > For Loop(LastIndex=11)
   > n2 = AddLine!(execute, self:Scriptable Tool Line Set Object Reference=Get Line Set() -> (Line Set:Scriptable Tool Line Set Object Reference)) -> (then, ReturnValue:Scriptable Tool Line Object Reference)
   > Add(
    TargetArray:Array of Scriptable Tool Line Object References=Get BoundingBoxLines()
    NewItem:"Scriptable Tool Line Object Reference (by ref)"=@n2
   )
   Outputs()
  ]
 }
 Init BBox {
  path=EventGraph.K2Node_Composite_8.Init BBox
  nodes [
   Inputs()
   > n1 = GetSelectedLevelActors(self=Get EditorActorSubsystem())
   > n2 = Set ActorsSelectedOnStartup(ActorsSelectedOnStartup:Array of Actor Object References=@n1) -> (Output_Get:Array of Actor Object References)
   > DrawBox!(execute, self, Actors:Array of Actor Object References=@n2) -> (then) {is=title K2Node_CallFunction}
   Outputs()
  ]
 }
 OnSelectionChanged {nodes [
   OnSelectionChanged!() -> (then, SelectionSet:Typed Element Selection Set Object Reference) {is=title K2Node_FunctionEntry}
   > n1 = GetSelectedLevelActors(self=Get EditorActorSubsystem())
   > Set FilteredSelectedActors(FilteredSelectedActors:Array of Actor Object References=@n1) -> (Output_Get:Array of Actor Object References)
   > n2 = For Each Loop(Array:Array of Actor Object References=Get GeneratedVolumesArray() -> (GeneratedVolumesArray:Array of Actor Object References)) -> (Array Element:Actor Object Reference)
   DrawBox(execute=@n2.Completed, Actors=Get FilteredSelectedActors() -> (FilteredSelectedActors:Array of Actor Object References))
   Remove Item!(execute=@n2.LoopBody, TargetArray:Array of Actor Object References=Get FilteredSelectedActors(), Item:"Actor Object Reference (by ref)"=@n2) -> (then, ReturnValue:bool)
  ]}
 OnVolumeGenerated {nodes [
   n1 = OnVolumeGenerated!() -> (then, GeneratedVolume:Actor Object Reference) {is=title K2Node_FunctionEntry}
   > n2 = Set Generated Volume(Generated Volume:Actor Object Reference=@n1) -> (Output_Get:Actor Object Reference)
   > Add(TargetArray:Array of Actor Object References=Get GeneratedVolumesArray(), NewItem:"Actor Object Reference (by ref)"=@n2)
   > SetSelectedLevelActors(self=Get EditorActorSubsystem(), ActorsToSelect=Get GeneratedVolumesArray())
  ]}
 SelectLastVolume {nodes [
   n1 = SelectLastVolume!() -> (then, GeneratedVolume:Actor Object Reference) {is=title K2Node_FunctionEntry}
   > Add(TargetArray=Get LastVolumeArray() -> (LastVolumeArray:Array of Actor Object References), NewItem=@n1)
   > SetSelectedLevelActors(self=Get EditorActorSubsystem(), ActorsToSelect=Get LastVolumeArray())
   > Clear(TargetArray:Array of Actor Object References=Get LastVolumeArray())
  ]}
 SelectPreviouslySelectedActors {nodes [
   SelectPreviouslySelectedActors() {is=title K2Node_FunctionEntry}
   > SetSelectedLevelActors(
    self=Get EditorActorSubsystem()
    ActorsToSelect=Get PreviouslySelectedActors!(self:EUW Audio Volume Creator Object Reference=Get Actions Area() -> (Actions Area:EUW Audio Volume Creator Object Reference)) -> (PreviouslySelectedActors:Array of Actor Object References)
   )
  ]}
 SetLineEdgeBBox {nodes [
   n1 = SetLineEdgeBBox!() -> (
    then
    LineNumber:int
    StartPointNumber:int
    EndPointNumber:int
    BoxLines:Array of Scriptable Tool Line Object References
    Corners:Array of Vectors
   ) {is=title K2Node_FunctionEntry}
   > SetLineEndPoints!(
    execute
    self:Scriptable Tool Line Object Reference="Get (a copy)"(Array:Array of Scriptable Tool Line Object References=@n1.BoxLines, Dimension 1=@n1) -> (Output:Scriptable Tool Line Object Reference)
    Start:Vector="Get (a ref)"(Array:Array of Vectors=@n1.Corners, Dimension 1=@n1.StartPointNumber) -> (Output:"Vector (by ref)")
    End:Vector="Get (a ref)"(Array=@n1.Corners, Dimension 1=@n1.EndPointNumber)
   ) -> (then)
  ]}
 UpdateBoxAppearance {nodes [
   n1 = UpdateBoxAppearance!() -> (then, Box Lines:Array of Scriptable Tool Line Object References, Color:Color Structure, Thickness:float=~, DepthBias:float=~) {is=title K2Node_FunctionEntry}
   > n2 = For Each Loop(Array:Array of Scriptable Tool Line Object References=@n1) -> (Array Element:Scriptable Tool Line Object Reference)
   > SetLineDepthBias!(execute, self:Scriptable Tool Line Object Reference=@n2, DepthBias:float=@n1.DepthBias) -> (then)
   > SetLineColor!(execute, self:Scriptable Tool Line Object Reference=@n2, Color:Color Structure=@n1.Color) -> (then)
   > SetLineThickness!(execute, self:Scriptable Tool Line Object Reference=@n2, Thickness:float=@n1.Thickness) -> (then)
  ]}
]
variables [
 Actions Area {
  type=EUW_AudioVolumeCreator_C
  default=&/BP_ST_AudioVolumeEditor.Default__BP_ST_AudioVolumeEditor_C:EUW_AudioVolumeCreator_C_0
 }
 ActorsSelectedOnStartup {type="array<Actor>"}
 As Audio Volume Creator Widget {type=EUW_AudioVolumeCreator_C, default=null}
 BoundingBoxLines {type="array<ScriptableToolLine>", default []}
 FilteredSelectedActors {type="array<Actor>"}
 Generated Volume {type=Actor, default=null}
 GeneratedVolumesArray {type="array<Actor>"}
 LastVolumeArray {type="array<Actor>"}
 Line Set {type=ScriptableToolLineSet, default=null}
]
